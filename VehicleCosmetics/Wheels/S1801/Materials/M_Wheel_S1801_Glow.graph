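# Translucent, unlit, two-sided glow material for the S1801 vehicle wheel cosmetic, usable on skeletal meshes.
# A glow texture is scaled by Glow Size, with a Panner fed a size-derived constant as Time to offset it toward center.
# Its red channel, optionally sine-pulsed through a static switch, drives tinted emissive and opacity.
# Rank Icon alpha is added only to the clamped opacity, with its own adjustment.
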Material /VehicleCosmetics/Wheels/S1801/Materials/M_Wheel_S1801_Glow
graph {
 properties {
  bAutomaticallySetUsageInEditor=false
  bUseMaterialAttributes=true
  bUsedWithMeshDeformer=true
  bUsedWithSkeletalMesh=true
  blendMode=BLEND_Translucent
  shadingModel=MSM_Unlit
  twoSided=true
  usedShadingModels=Unlit
 }
 nodes [
  n1 = VectorParameter() {parameterName=Glow Color, defaultValue=<rgba 1 0.5029 0 1>}
  n2 = TextureCoordinate()
  n3 = ScalarParameter() {defaultValue=1, parameterName=Glow Size}
  n4 = OneMinus(None=@n3)
  n5 = TextureSampleParameter2D(UVs=Panner(Coordinate=Multiply(A=@n2, B=@n3), Time=Divide(A=@n4)) {speedX=1, speedY=1}) {parameterName=Glow, texture=/VehicleCosmetics/Wheels/_SharedTextures/FX/Glow_N_T}
  n6 = Time()
  n7 = Multiply(A=Add(B=Sine(None=@n6) {period=2}) {constA=1}) {constB=0.5}
  n8 = StaticSwitchParameter(True=Multiply(A=@n5.R, B=Add(A=@n7) {constA=1}) {constB=0.5}, False=@n5.R) {parameterName=Pulse}
  NamedRerouteDeclaration(None=Multiply(A=Multiply(A=@n1.RGB, B=@n8), B=ScalarParameter() {defaultValue=1, parameterName=Glow Color Intensity})) {name=Emissive, nodeColor=<rgba 1 0.0234 0 1>}
  n9 = Multiply(A=@n8, B=ScalarParameter() {defaultValue=1, parameterName=Glow Opacity Intensity})
  n10 = Multiply(
   A=TextureSampleParameter2D() {parameterName=Rank Icon, texture=/VehicleCosmetics/Wheels/Sr09T1/Textures/T_Sr09t1_RankIcon_Bronze}.A
   B=ScalarParameter() {defaultValue=1, parameterName=Opacity Adjust}
  )
  NamedRerouteDeclaration(None=Clamp(None=Add(A=@n9, B=@n10))) {name=Opcaity, nodeColor=<rgba 1 0 0.5156 1>}
  SetMaterialAttributes!(MaterialAttributes, Emissive Color=NamedRerouteUsage(), Opacity=NamedRerouteUsage()) {attributeSetTypes [#1, #2]}
 ]
}
