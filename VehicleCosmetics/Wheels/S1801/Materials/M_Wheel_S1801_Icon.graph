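# Masked, unlit, two-sided material for the rank icon on the S1801 cosmetic vehicle wheel, usable on skeletal meshes.
# The Rank Icon texture, scaled by Diffuse Intensity, is added to a tinted, intensity-scaled Glow texture to form emissive color.
# Glow Size scales the glow UVs about their center, using a Panner fed a computed time as a static offset rather than animation.
# The icon's alpha drives the opacity mask, so it also clips the glow.
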Material /VehicleCosmetics/Wheels/S1801/Materials/M_Wheel_S1801_Icon
graph {
 properties {
  bAutomaticallySetUsageInEditor=false
  bUseMaterialAttributes=true
  bUsedWithMeshDeformer=true
  bUsedWithSkeletalMesh=true
  blendMode=BLEND_Masked
  shadingModel=MSM_Unlit
  twoSided=true
  usedShadingModels=Unlit
 }
 nodes [
  n1 = ScalarParameter() {defaultValue=1, parameterName=Glow Size}
  n2 = Multiply(A=TextureCoordinate(), B=@n1)
  n3 = Divide(A=OneMinus(None=@n1))
  n4 = Multiply(
   A=VectorParameter() {parameterName=Glow Color, defaultValue=<rgba 1 0.5029 0 1>}.RGB
   B=TextureSampleParameter2D(UVs=Panner(Coordinate=@n2, Time=@n3) {speedX=1, speedY=1}) {parameterName=Glow, texture=/VehicleCosmetics/Wheels/_SharedTextures/FX/Glow_N_T}.R
  )
  n5 = ScalarParameter() {defaultValue=1, parameterName=Glow Color Intensity}
  n6 = TextureSampleParameter2D() {parameterName=Rank Icon, texture=/VehicleCosmetics/Wheels/Sr09T1/Textures/T_Sr09t1_RankIcon_Bronze}
  n7 = ScalarParameter() {defaultValue=2, parameterName=Diffuse Intensity}
  NamedRerouteDeclaration(None=Add(A=Multiply(A=@n4, B=@n5), B=Multiply(A=@n6.RGB, B=@n7))) {name=Emissive, nodeColor=<rgba 1 0.0234 0 1>}
  SetMaterialAttributes!(MaterialAttributes, Emissive Color=NamedRerouteUsage(), Opacity Mask=@n6.A) {attributeSetTypes [#1, #2]}
 ]
}
